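# Niagara impact effect for the Galaxy skin of the GoldenValley pickaxe, bursting cyan smoke and flying petals.
# Galaxy_Smoke spawns short-lived sprites that grow and fade; Petals001 launches sprites outward, slowed by drag under gravity and curl noise.
# Petals spin and fake flutter through sine-driven width scaling after launch.
# Front-end modifiers and a preview bool adjust scale, speed and petal count; the other eight emitters list no renderers.
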NiagaraSystem /BRCosmetics/Weapons/FORT_Melee/Pickaxe_GoldenValley/Skins/GoldenValley_Galaxy/FX/NS_Pickaxe_GoldenValley_Impact_Galaxy
dependencies {
 usedDataInterfaces [&NiagaraDataInterfaceCurve, &NiagaraDataInterfaceVector2DCurve]
 usedDynamicInputs [&MakeCustomFloatFromBool, &MakeCustomVectorFromBool]
 usedModules [
  &self:SystemSpawnScript.SystemScriptSource.SystemScriptGraph.NiagaraNodeAssignment_0.SetVariables_#1
  &self:SystemSpawnScript.SystemScriptSource.SystemScriptGraph.NiagaraNodeAssignment_2.SetVariables_#2
  &self:SystemSpawnScript.SystemScriptSource.SystemScriptGraph.NiagaraNodeAssignment_3.SetVariables_#3
  &SystemState
 ]
}
emitters {
 Dust {rendererClasses [], renderers []}
 FlareAndLight {rendererClasses [], renderers []}
 Galaxy_Smoke {
  emitterUpdateScript [
   &EmitterState(Life Cycle Mode=::NewEnumerator1, Loop Behavior=::NewEnumerator1, Loop Duration=0.5, UseLoopDelay=true)
   &SpawnBurst_Instantaneous(
    Spawn Count=&RandomRangeInt(Minimum=5, Maximum=8)
    Use Spawn Probability=false
    UseLoopCountLimit=false
   )
  ]
  particleSpawnScript [
   &InitializeParticle(
    Color=&MakeLinearColorFromVectorAndFloat(
     "Vector (RGB)"=&MakeVectorFromLinearColorRGB(Color=<rgba 0.209 0.9868 1 0.5>)
     "Float (Alpha)"=&RandomRangeFloat(Minimum=0.25, Maximum=0.5)
    )
    Lifetime=&RandomRangeFloat(Minimum=0.3, Maximum=0.4)
    Sprite Rotation Mode=::NewEnumerator2
    Sprite Size Mode=::NewEnumerator3
    Uniform Sprite Size=&RandomRangeFloat(Minimum=85, Maximum=110)
    UsePositionOffset=false
    Write Material Random=true
   )
   &ShapeLocation(
    MaskSpawnGroup=false
    Override Seed=false
    Sphere Radius=&Multiply_Float(A=5, B=&FN_FrontEndFloatModifier(Gameplay=1, Front End=$System.FrontendScale))
    UseOwnerScale=true
   )
   &CameraOffset(Camera Offset Amount=10)
  ]
  particleUpdateScript [
   &ParticleState()
   &Drag(
    Drag=&FloatFromCurve(
     FloatCurve=DataInterface() {propertyValues {Curve {keys [
         {arriveTangent=0.8664, interpMode=RCIM_Cubic, leaveTangent=0.8664, tangentMode=RCTM_User, time=0, value=0.375}
         {interpMode=RCIM_Cubic, tangentMode=RCTM_User, time=1, value=1}
        ]}}}
     Scale Curve=&RandomRangeFloat(Minimum=12, Maximum=15)
    )
    Use Linear Drag=true
    Use Rotational Drag=false
   )
   &SolveForcesAndVelocity(Clamp Velocity=false, Limit Acceleration=false)
   &ScaleColor(
    Scale Alpha=&FloatFromCurve(FloatCurve=DataInterface() {propertyValues {Curve {keys [
         {arriveTangent=-0.0405, interpMode=RCIM_Cubic, leaveTangent=-0.0405, tangentMode=RCTM_User, time=0, value=1}
         {arriveTangent=-2.127, interpMode=RCIM_Cubic, leaveTangent=-2.127, tangentMode=RCTM_User, time=1, value=0}
        ]}}})
    Scale RGB=&VectorFromFloat(Value=&FloatFromCurve(
      FloatCurve=DataInterface() {propertyValues {Curve {keys [
          {arriveTangent=-2.4245, interpMode=RCIM_Cubic, leaveTangent=-2.4245, tangentMode=RCTM_User, time=0, value=1}
          {interpMode=RCIM_Cubic, tangentMode=RCTM_User, time=1, value=0}
         ]}}}
      Scale Curve=3
     ))
    ScaleA=true
    ScaleRGB=false
   )
   &ScaleSpriteSize(
    Uniform Curve Scale=&FN_FrontEndFloatModifier(Gameplay=1, Front End=$System.FrontendScale)
    Uniform Curve Sprite Scale=DataInterface() {propertyValues {Curve {keys [
        {arriveTangent=1.8336, interpMode=RCIM_Cubic, leaveTangent=1.8336, tangentMode=RCTM_User, time=0, value=0}
        {interpMode=RCIM_Cubic, tangentMode=RCTM_User, time=0.9, value=1}
       ]}}}
   )
  ]
  rendererClasses [&NiagaraSpriteRendererProperties]
  renderers [{rendererClass=&NiagaraSpriteRendererProperties}]
 }
 Petals001 {
  emitterUpdateScript [
   &EmitterState(Life Cycle Mode=::NewEnumerator1, Loop Behavior=::NewEnumerator1, Loop Delay=0.025, Loop Duration=0.5, UseLoopDelay=true)
   &SpawnBurst_Instantaneous(Spawn Count=200, Use Spawn Probability=false, UseLoopCountLimit=false) {enabled=false, module=SpawnBurst_Instantaneous001}
   &SpawnBurst_Instantaneous(
    Spawn Count=&MakeCustomIntFromBool(
     Bool=$User.bIsFrontEndPreview
     True INT=&RandomRangeInt(Minimum=5, Maximum=10)
     False INT=&RandomRangeInt(Minimum=10, Maximum=15)
    )
    Use Spawn Probability=false
    UseLoopCountLimit=false
   )
  ]
  particleSpawnScript [
   &InitializeParticle(
    Lifetime=&RandomRangeFloat(Minimum=1, Maximum=1.5)
    Sprite Rotation Mode=::NewEnumerator2
    Sprite Size=&Multiply_Vector2D_ByFloat(
     Float=&RandomRangeFloat(Minimum=0.4, Maximum=0.6)
     Vector2D=&Multiply_Vector2D_ByFloat(Float=&FN_FrontEndFloatModifier(Gameplay=1, Front End=0.75), Vector2D=<26 10>)
    )
    Sprite Size Mode=::NewEnumerator4
    UsePositionOffset=false
    Write Material Random=true
   ) {module=InitializeParticle001}
   &ShapeLocation(
    Additional Quaternion Rotation=true
    Additional Yaw / Pitch / Roll=<-90 0 0>
    Hemisphere Distribution=<0.5 1>
    MaskSpawnGroup=false
    Override Seed=false
    Rotation Angle Type:ENiagara_AngleInput=::NewEnumerator0
    Rotation Mode=::NewEnumerator3
    Sphere Radius=0.01
    UseOwnerScale=true
   )
   &AddVelocity(
    Rotation Coordinate Space=::Simulation
    Velocity=&Multiply_VectorByFloat(Vector=$Particles.ShapeLocation.ShapeNormal, Float=&RandomRangeFloat(Minimum=1000, Maximum=1500))
    Velocity Speed Scale=&FN_FrontEndFloatModifier(Gameplay=1, Front End=0.8)
   )
   /SetVariables(Particles.stableGravity:NiagaraFloat=&UniformRangedFloat(Minimum=-150, Maximum=-200)) {isSetParameters=true}
   /SetVariables(Particles.stableRotationRate:NiagaraFloat=&UniformRangedFloat(Minimum=50, Maximum=200)) {isSetParameters=true}
  ]
  particleUpdateScript [
   &ParticleState()
   &GravityForce(Gravity=&MakeVector(X=0, Y=0, Z=$Particles.Initial.stableGravity))
   &Drag(
    Drag=&Multiply_Float(
     A=&UniformRangedFloat(Minimum=1.2, Maximum=1.5)
     B=&FloatFromCurve(
      FloatCurve=DataInterface() {propertyValues {Curve {keys [
          {arriveTangent=-5.3551, interpMode=RCIM_Cubic, leaveTangent=-5.3551, tangentMode=RCTM_User, time=0, value=0}
          {arriveTangent=-138.8873, interpMode=RCIM_Cubic, leaveTangent=-138.8873, tangentMode=RCTM_User, time=0.05, value=7}
          <at 0.2 0.5>
         ]}}}
      Scale Curve=4
     )
    )
    Use Linear Drag=true
    Use Rotational Drag=false
   )
   /Niagara/Modules/Update/Forces/CurlNoiseForce(Noise Strength=300, Pan Noise=false)
   &SolveForcesAndVelocity(Clamp Velocity=false, Limit Acceleration=false)
   &SpriteRotationRate(Rotation Rate=$Particles.Initial.stableRotationRate)
   &ScaleSpriteSize(Scale Factor=&Lerp_Vector2(
     A=&Multiply_Vector2(
      A=&MakeVector2D(
       X=&Multiply_Float(
        A=1
        B=&Sine(
         Normalized Angle=$Emitter.Age
         Period=&Multiply_Float(A=0.7, B=&Add_Float(A=$Particles.Initial.MaterialRandom, B=0.65))
         Scale=1
         Bias=0
        )
       )
       Y=1
      )
      B=&Vector2DFromFloat(Value=&FloatFromCurve(FloatCurve=DataInterface() {propertyValues {Curve {keys [<at 0 1.1>,<at 0.9 1.1>,<at 1 0>]}}}))
     )
     B=<1 1>
     Alpha=&FloatFromCurve(FloatCurve=DataInterface() {propertyValues {Curve {keys [<at 0.0574 1>,<at 0.1545 0>]}}})
    ))
  ]
  rendererClasses [&NiagaraSpriteRendererProperties]
  renderers [
   {rendererClass=&NiagaraSpriteRendererProperties}
   {rendererClass=&NiagaraSpriteRendererProperties, rendererIndex=1}
  ]
 }
 SmokePuff {rendererClasses [], renderers []}
 Sparks001 {rendererClasses [], renderers []}
 SpikeGreen {rendererClasses [], renderers []}
 SpikePink {rendererClasses [], renderers []}
 StarSpike {rendererClasses [], renderers []}
 Wave001 {rendererClasses [], renderers []}
}
summary {
 systemName=NS_Pickaxe_GoldenValley_Impact_Galaxy
 emitters [
  {emitterName=Galaxy_Smoke, rendererClasses [&NiagaraSpriteRendererProperties]}
  {emitterName=SpikePink, rendererClasses []}
  {emitterName=SpikeGreen, rendererClasses []}
  {emitterName=StarSpike, rendererClasses []}
  {emitterName=SmokePuff, rendererClasses []}
  {emitterName=FlareAndLight, rendererClasses []}
  {emitterName=Sparks001, rendererClasses []}
  {emitterName=Wave001, rendererClasses []}
  {emitterName=Dust, rendererClasses []}
  {emitterName=Petals001, rendererClasses [&NiagaraSpriteRendererProperties]}
 ]
 userVariables [
  User.bIsFrontEnd {type=NiagaraBool}
  User.bIsFrontEndPreview {type=NiagaraBool}
 ]
}
